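# Physics asset for the PawJasmineSword pickaxe cosmetic's axe mesh, defining collision bodies and a joint for physically simulated motion.
# A tiny capsule on b_AR_Root and a long, angled, offset capsule on the dyn_head bone form the two bodies.
# A single constraint links dyn_head to the root with swing and twist angle limits, bounding how far the head can move.
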PhysicsAsset /BRCosmetics/Weapons/FORT_Melee/Pickaxe_PawJasmineSword/Meshes/PawJasmineSword_Axe_FE_Physics
bodies {
 b_AR_Root [{shapeType=Capsule, length=0.505, radius=0.505, rotation=<rot 0 -90 0>}]
 dyn_head [{shapeType=Capsule, length=45.3991, radius=9, shapeName=dyn_head_capsule, center=<14 0 -14>, rotation=<rot -30 0 0>}]
}
constraints [{bone1Name=dyn_head, bone2Name=b_AR_Root, swing1LimitDegrees=35, swing2LimitDegrees=25, twistLimitDegrees=10}]
