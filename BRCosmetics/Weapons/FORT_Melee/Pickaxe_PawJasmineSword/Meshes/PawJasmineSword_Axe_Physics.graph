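# Physics asset for the PawJasmineSword axe mesh, a cosmetic melee pickaxe.
# It has two capsule bodies: a tiny one on the b_AR_Root bone and a larger, offset and rotated one on dyn_head.
# A single constraint ties dyn_head to the root with narrow swing and twist limits, allowing only slight angular movement of the head.
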PhysicsAsset /BRCosmetics/Weapons/FORT_Melee/Pickaxe_PawJasmineSword/Meshes/PawJasmineSword_Axe_Physics
bodies {
 b_AR_Root [{shapeType=Capsule, length=0.505, radius=0.505, rotation=<rot 0 -90 0>}]
 dyn_head [{
   shapeType=Capsule
   length=38.1706
   radius=11.7427
   shapeName=dyn_head_capsule
   center=<12 -0.0153 -12.1607>
   rotation=<rot 0.2381 -90 23.731>
  }]
}
constraints [{bone1Name=dyn_head, bone2Name=b_AR_Root, swing1LimitDegrees=10, swing2LimitDegrees=10, twistLimitDegrees=7}]
