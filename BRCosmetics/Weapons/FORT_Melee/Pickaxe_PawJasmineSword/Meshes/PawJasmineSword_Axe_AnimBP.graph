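# Animation Blueprint for the PawJasmineSword pickaxe that adds physics-driven secondary motion to its skeleton.
# The cached input pose feeds two partially weighted RigidBody nodes with separate physics assets and gravity overrides.
# bIsInFrontEnd selects between the front-end and in-game simulation results.
# The update event copies the matching preset into a Sim Space Settings variable, which neither RigidBody node visibly reads.
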AnimBlueprint /BRCosmetics/Weapons/FORT_Melee/Pickaxe_PawJasmineSword/Meshes/PawJasmineSword_Axe_AnimBP
target_skeleton=&/PawJasmineSword_Axe_Skeleton
graphs [
 AnimGraph {nodes [
   n1 = RigidBody!(
    ComponentPose=Local To Component(LocalPose=Use cached pose 'Pose'())
    Alpha=0.85
    ExternalForce
    SimSpaceSettings:Sim Space Settings Structure=(WorldAlpha=0,VelocityScaleZ=1,DampingAlpha=1,MaxLinearVelocity=10000,MaxAngularVelocity=10000,MaxLinearAcceleration=10000,MaxAngularAcceleration=10000,ExternalLinearDragV=(X=0,Y=0,Z=0),ExternalLinearVelocity=(X=0,Y=0,Z=0),ExternalAngularVelocity=(X=0,Y=0,Z=0))
   ) -> (Pose) {
    bOverrideWorldGravity=true
    overridePhysicsAsset=&/PawJasmineSword_Axe_Physics
    alphaBoolBlend {blendOption=Linear}
    baseBoneRef {boneName=b_AR_Root}
    componentAppliedLinearAccClamp=<5000 5000 5000>
    overrideWorldGravity {z=-980}
    simSpaceSettings {dampingAlpha=0, maxAngularAcceleration=240, maxAngularVelocity=8, maxLinearAcceleration=240, maxLinearVelocity=850, worldAlpha=1}
   }
   n2 = RigidBody(
    ComponentPose=Local To Component(LocalPose=Use cached pose 'Pose'())
    Alpha=0.85
    SimSpaceSettings=(WorldAlpha=0,VelocityScaleZ=1,DampingAlpha=1,MaxLinearVelocity=10000,MaxAngularVelocity=10000,MaxLinearAcceleration=10000,MaxAngularAcceleration=10000,ExternalLinearDragV=(X=0,Y=0,Z=0),ExternalLinearVelocity=(X=0,Y=0,Z=0),ExternalAngularVelocity=(X=0,Y=0,Z=0))
   ) {
    bOverrideWorldGravity=true
    overridePhysicsAsset=&/PawJasmineSword_Axe_FE_Physics
    alphaBoolBlend {blendOption=Linear}
    baseBoneRef {boneName=b_AR_Root}
    componentAppliedLinearAccClamp=<5000 5000 5000>
    overrideWorldGravity {z=-1200}
    simSpaceSettings {dampingAlpha=0, maxAngularAcceleration=240, maxAngularVelocity=8, maxLinearAcceleration=240, maxLinearVelocity=850, worldAlpha=1}
   }
   Output Pose(Result=Blend Poses by bool(BlendPose_0=Component To Local(ComponentPose=@n2), BlendPose_1=Component To Local(ComponentPose=@n1), bActiveValue=Get bIsInFrontEnd()) {blendType=Linear}) {layerGroup=None}
   AnimGraphNode_SubInput_0 = Input Pose()
   Save cached pose 'Pose'(Pose=@AnimGraphNode_SubInput_0) {cacheName=Pose}
  ]}
 EventGraph {nodes [
   Event BlueprintUpdateAnimation()
   > n1 = Branch(Condition=Get bIsInFrontEnd())
   > Set Sim Space Settings(Sim Space Settings:Sim Space Settings Structure=Get Sim Space Settings Front End() -> (Sim Space Settings Front End:Sim Space Settings Structure)) -> (Output_Get:Sim Space Settings Structure)
   Set Sim Space Settings(execute=@n1.else, Sim Space Settings=Get Sim Space Settings in zone() -> (Sim Space Settings in zone:Sim Space Settings Structure))
   TryGetPawnOwner()
  ]}
]
variables [
 Sim Space Settings {
  type=SimSpaceSettings
  default {
   dampingAlpha=0
   maxAngularAcceleration=240
   maxAngularVelocity=8
   maxLinearAcceleration=240
   maxLinearVelocity=850
   velocityScaleZ=1
   worldAlpha=1
   externalLinearDragV=<0 0 0>
  }
 }
 Sim Space Settings Front End {
  type=SimSpaceSettings
  default {
   dampingAlpha=0
   maxAngularAcceleration=35
   maxAngularVelocity=7
   maxLinearAcceleration=3500
   maxLinearVelocity=700
   velocityScaleZ=1
   worldAlpha=1
   externalLinearDragV=<0.1 0.1 0.1>
  }
 }
 Sim Space Settings in zone {
  type=SimSpaceSettings
  default {
   dampingAlpha=0
   maxAngularAcceleration=240
   maxAngularVelocity=8
   maxLinearAcceleration=240
   maxLinearVelocity=850
   velocityScaleZ=1
   worldAlpha=1
   externalLinearDragV=<0 0 0>
  }
 }
]
